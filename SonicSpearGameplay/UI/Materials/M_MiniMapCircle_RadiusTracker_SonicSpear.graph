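# Translucent UI material that draws the Sonic Spear radius tracker circle on the minimap, placed by a center-and-radius vector parameter.
# A time-driven angle gradient, scaled by a radar on/off parameter, forms a sweep over the safe-zone-tested fill.
# A rotating segmented outline combines with that fill under an inverted radial gradient, plus a smoothstepped, edge-masked center icon.
# A static switch optionally rescales UVs using the minimap parameter collection's ScaleFactor.
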Material /SonicSpearGameplay/UI/Materials/M_MiniMapCircle_RadiusTracker_SonicSpear
graph {
 properties {blendMode=BLEND_Translucent, materialDomain=MD_UI, numCustomizedUVs=1, pixelDepthOffsetMode=PDOM_Legacy, usedShadingModels=Default Lit}
 nodes [
  n1 = VectorParameter() {desc=Color to use to draw the circle, group=EffectParams, parameterName=CircleColor, defaultValue=<rgba 0 1 0.7084 1>}
  n2 = VectorParameter() {
   desc="Circle Origin.X, Origin.Y, Radius (all multiplied by map scale).\r\nA is ignored"
   parameterName=CircleCenterAndRadius
   defaultValue=<rgba 0.5 0.5 1 1>
  }
  n3 = ComponentMask(None=@n2.RGB) {g=true, r=true}
  n4 = Subtract(A=TextureCoordinate(), B=@n3)
  n5 = Add(A=Divide(A=@n4, B=@n2.B)) {constB=0.5}
  n6 = Multiply(A=Time(), B=ScalarParameter() {defaultValue=0.2, parameterName=SweepSpeed})
  n7 = Fmod(A=@n6, B=Constant() {r=1})
  n8 = ConstantBiasScale(None=Multiply(
    A=ScalarParameter() {parameterName="Radar on off (0-1)"}
    B=MaterialFunctionCall!(UVs=@n5, Rotation=@n7) {materialFunction=/FortUILibrary/MaterialFunctions/Gradients/MF_UI_AngleGradient}.Result
   ) {constB=0}) {bias=0.6}
  n9 = TextureObjectParameter() {group=IconParameters, parameterName=Icon Texture, sortPriority=1, texture=/Game/Athena/Items/Accolades/Materials/Textures/Spree_Alt128px}
  n10 = ScalarParameter() {defaultValue=0.3, group=IconParameters, parameterName=IconScale}
  n11 = Add(A=Divide(A=@n4, B=Multiply(A=@n2.B, B=@n10))) {constB=0.5}
  n12 = Constant() {r=0.5}
  n13 = MaterialFunctionCall(Float2=@n11) {materialFunction=&BreakOutFloat2Components}
  n14 = SphereMask(A=@n12, B=@n13.R) {attenuationRadius=0.5}
  n15 = SphereMask(A=@n13.G, B=@n12) {attenuationRadius=0.5}
  n16 = ScalarParameter() {defaultValue=0.2, group=IconParameters, parameterName=IconEdgeSoftness}
  n17 = ScalarParameter() {defaultValue=0.18, group=IconParameters, parameterName=IconEdge}
  n18 = MaterialFunctionCall(
   Alpha=Multiply(
    A=TextureSample(UVs=@n11, Tex=@n9) {samplerSource=SSM_Clamp_WorldGroupSettings, texture=&Spree_Alt128px}.R
    B=Multiply(A=@n14, B=@n15)
   )
   Min=@n17
   Max=Add(A=Multiply(A=@n17, B=@n16), B=@n17)
  ) {materialFunction=&SmoothStep}
  LinearInterpolate(A=Power(Base=@n1.RGB, Exp=@n8), B=@n1.RGB, Alpha=@n18.Result)
  n19 = MaterialFunctionCall(
   UVs=@n5
   Radius=ScalarParameter() {defaultValue=1, parameterName=RadialRadius}
   Density=ScalarParameter() {defaultValue=2, parameterName=RadialDensity}
  ) {materialFunction=&RadialGradientExponential}
  n20 = Constant()
  n21 = ComponentMask(None=@n2.RGB) {b=true, g=true, r=true}
  n22 = Multiply(
   A=@n8
   B=Multiply(
    A=ScalarParameter() {defaultValue=1, parameterName=CircleFade}
    B=MaterialFunctionCall!(InRadiusAdditive=@n20, InSafeZoneData=@n21) {materialFunction=/Game/Athena/HUD/MiniMap/MF_SafeZoneTest}.Threshold
   )
  )
  n23 = Divide(B=MaterialFunctionCall() {materialFunction=&GetUserInterfaceUV}.Pixel Size) {constA=900, constB=1024}
  n24 = ScalarParameter() {defaultValue=-0.05, group=OutlineParameters, parameterName=SegmentSpeed}
  n25 = Multiply(
   A=MaterialFunctionCall!(
    Center=@n3
    Radius=@n2.B
    Thickness=ScalarParameter() {defaultValue=2, group=OutlineParameters, parameterName=Thickness_Outline}
    Falloff=ScalarParameter() {defaultValue=1, group=OutlineParameters, parameterName=Falloff}
    PixelToUVDensity=Multiply(A=@n23) {constB=0.001}
   ) {materialFunction=/Game/Athena/HUD/MiniMap/MF_MinimapCircleMaker2}.Circle
   B=MaterialFunctionCall!(
    EdgeSmooth
    Repeat=ScalarParameter() {defaultValue=8, group=OutlineParameters, parameterName=EdgeSegments}
    CenterOffset
    Rotation=Multiply(A=@n6, B=@n24)
    UV=@n5
    SegmentSpacing=ScalarParameter() {defaultValue=0.002, group=OutlineParameters, parameterName=SegmentSpacing}
   ) {materialFunction=/Game/Athena/HUD/reticles/MF_UI_RadialSegments_DEPRECATED}.Result
  )
  Max(
   A=Multiply(A=@n18.Result, B=ScalarParameter() {defaultValue=0.1, group=IconParameters, parameterName=IconBoost})
   B=Multiply(A=OneMinus(None=@n19.RadialGradientExponential), B=Max(A=@n22, B=@n25))
  )
  n26 = TextureCoordinate()
  n27 = CollectionParameter() {collection=/Game/Athena/HUD/MiniMap/MPC_Minimap, parameterName=ScaleFactor}
  n28 = OneMinus(None=@n27)
  StaticSwitchParameter(True=Add(A=Multiply(A=@n26, B=@n27) {constB=2}, B=Multiply(A=@n28) {constB=0.5}), False=@n26) {parameterName=UseFullScreenScale}
 ]
}
