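# Translucent UI material that draws the Sonic Spear alert donut: a donut icon with five short line segments around its left and right sides.
# The donut texture's alpha is sampled with UVs scaled from the center and rotated, with size and rotation parameters.
# MF_UI_LineSegment builds the lines procedurally from fixed endpoints, routed through named reroutes.
# The lines, weighted by a shared opacity, are added to the icon alpha, then scaled by Global Opacity.
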Material /SonicSpearGameplay/UI/Materials/M_UI_SonicSpear_Alert_Donut
graph {
 properties {bAutomaticallySetUsageInEditor=false, blendMode=BLEND_Translucent, materialDomain=MD_UI, usedShadingModels=Default Lit}
 nodes [
  n1 = AppendVector(A=ScalarParameter() {defaultValue=0.5, parameterName=Icon Width}, B=ScalarParameter() {defaultValue=0.5, parameterName=Icon Height})
  n2 = TextureSample(UVs=MaterialFunctionCall(
    UVs=MaterialFunctionCall(Texture Scale=@n1) {materialFunction=&ScaleUVsByCenter}.UVs
    "Rotation Angle (0-1)"=ScalarParameter() {defaultValue=1, parameterName=Rotate icon}
   ) {materialFunction=&CustomRotator}.Rotated Values) {texture=/SonicSpearGameplay/UI/Textures/T_UI_SonicSpear_Donut}
  n3 = ScalarParameter() {defaultValue=1, parameterName=Lines opacity}
  n4 = Multiply(A=NamedRerouteUsage(), B=@n3)
  n5 = NamedRerouteUsage()
  n6 = Add(A=Add(A=Add(A=@n2.A, B=@n4), B=Multiply(A=@n5, B=@n3)), B=Multiply(A=NamedRerouteUsage(), B=@n3))
  n7 = Multiply(A=NamedRerouteUsage(), B=@n3)
  n8 = NamedRerouteUsage()
  Multiply(A=Add(A=Add(A=@n6, B=@n7), B=Multiply(A=@n8, B=@n3)), B=ScalarParameter() {defaultValue=1, parameterName=Global Opacity})
  NamedRerouteDeclaration(None=MaterialFunctionCall!(
    UVs
    PointA=Constant2Vector() {r=-0.9}
    PointB=Constant2Vector() {g=0.1, r=-0.55}
    Thickness=ScalarParameter() {defaultValue=0.025, parameterName=Line thickness}
    OutlineThickness
    OutlineAlignment
    GlowMin
    GlowMax
   ) {materialFunction=/FortUILibrary/MaterialFunctions/Shapes/MF_UI_LineSegment}.Fill) {name=Line2, nodeColor=<rgba 0 1 0.5547 1>}
  NamedRerouteDeclaration(None=MaterialFunctionCall(
    PointA=Constant2Vector() {g=-0.25, r=0.75}
    PointB=Constant2Vector() {g=-0.1, r=0.55}
    Thickness=ScalarParameter() {defaultValue=0.025, parameterName=Line thickness}
   ) {materialFunction=&MF_UI_LineSegment}.Fill) {name=Line3, nodeColor=<rgba 0 0 0 1>}
  NamedRerouteDeclaration(None=MaterialFunctionCall(
    PointA=Constant2Vector() {g=0.1, r=0.85}
    PointB=Constant2Vector() {g=0.15, r=0.55}
    Thickness=ScalarParameter() {defaultValue=0.025, parameterName=Line thickness}
   ) {materialFunction=&MF_UI_LineSegment}.Fill) {name=Line4, nodeColor=<rgba 0 0 0 1>}
  NamedRerouteDeclaration(None=MaterialFunctionCall(
    PointA=Constant2Vector() {g=-0.3, r=-0.75}
    PointB=Constant2Vector() {g=-0.15, r=-0.55}
    Thickness=ScalarParameter() {defaultValue=0.025, parameterName=Line thickness}
   ) {materialFunction=&MF_UI_LineSegment}.Fill) {name=Line1, nodeColor=<rgba 0 0 0 1>}
  NamedRerouteDeclaration(None=MaterialFunctionCall(
    PointA=Constant2Vector() {g=0.5, r=0.8}
    PointB=Constant2Vector() {g=0.4, r=0.45}
    Thickness=ScalarParameter() {defaultValue=0.025, parameterName=Line thickness}
   ) {materialFunction=&MF_UI_LineSegment}.Fill) {name=Line5, nodeColor=<rgba 0 0 0 1>}
 ]
}
