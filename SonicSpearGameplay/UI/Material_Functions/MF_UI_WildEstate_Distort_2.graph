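# UI material function that produces a small animated noise-driven two-component offset, named as a Wild Estate distortion.
# The UV input is panned by SpeedX and SpeedY to sample the NoiseTexture parameter, a Perlin noise by default.
# The red channel fills both components and is scaled by Intensity, giving an uncentered diagonal offset that is not added to the UVs here.
# Unconnected inputs fall back to texture coordinates or same-named scalar parameters.
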MaterialFunction /SonicSpearGameplay/UI/Material_Functions/MF_UI_WildEstate_Distort_2
graph {nodes [
  n1 = TextureCoordinate()
  n2 = FunctionInput(Preview=ScalarParameter() {defaultValue=-1, parameterName=SpeedX}) {bUsePreviewValueAsDefault=true, inputName=SpeedX, inputType=FunctionInput_Scalar, sortPriority=1}
  n3 = FunctionInput(Preview=ScalarParameter() {defaultValue=1, parameterName=SpeedY}) {bUsePreviewValueAsDefault=true, inputName=SpeedY, inputType=FunctionInput_Scalar, sortPriority=2}
  n4 = TextureSampleParameter2D(UVs=Panner(
    Coordinate=FunctionInput(Preview=@n1) {bUsePreviewValueAsDefault=true, inputName=UV, inputType=FunctionInput_Vector2}
    Speed=AppendVector(A=@n2, B=@n3)
   ) {speedX=-1, speedY=2}) {parameterName=NoiseTexture, texture=&/T_UI_RTL_PerlinNoise2}
  n5 = ScalarParameter() {defaultValue=0.052, parameterName=Intensity}
  FunctionOutput(None=Multiply(
    A=AppendVector(A=@n4.R, B=@n4.R)
    B=FunctionInput(Preview=@n5) {bUsePreviewValueAsDefault=true, inputName=Intensity, inputType=FunctionInput_Scalar, sortPriority=5}
   ) {constB=0.056})
 ]}
