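# Translucent UI material for the Sonic Spear alert icon, showing a character texture over a colored circular badge.
# Two MF_UI_Circle calls share a size parameter and supply the fill, outline and thinner accent ring.
# The icon texture is offset and scaled about its center through position and size parameters.
# Named reroutes feed lerp chains that layer color and opacity, and a distortion function builds an offset UV set.
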Material /SonicSpearGameplay/UI/Materials/M_UI_SonicSpear_Alert_Icon
graph {
 properties {bAutomaticallySetUsageInEditor=false, blendMode=BLEND_Translucent, materialDomain=MD_UI, usedShadingModels=Default Lit}
 nodes [
  n1 = NamedRerouteUsage()
  n2 = NamedRerouteUsage()
  LinearInterpolate(A=LinearInterpolate(A=NamedRerouteUsage(), B=@n1, Alpha=@n1), B=@n2, Alpha=@n2)
  n3 = VectorParameter() {parameterName=BackroundColor, defaultValue=<rgba 0.6795 0.0369 0.2384 1>}
  n4 = NamedRerouteUsage()
  LinearInterpolate(
   A=LinearInterpolate(
    A=LinearInterpolate(A=@n3.RGB, Alpha=@n4) {constB=0}
    B=VectorParameter() {parameterName=OutlineColor, defaultValue=<rgba 0.9823 0.7084 0 1>}.RGB
    Alpha=NamedRerouteUsage()
   )
   B=NamedRerouteUsage()
   Alpha=NamedRerouteUsage()
  )
  n5 = NamedRerouteUsage()
  n6 = ScalarParameter() {defaultValue=0.75, parameterName=CircleSize}
  n7 = MaterialFunctionCall(UVs=@n5, Size=@n6, OutlineThickness=Constant() {r=0.06}) {materialFunction=&MF_UI_Circle}
  NamedRerouteDeclaration(None=@n7.Fill) {name=Circle, nodeColor=<rgba 0 0 0 1>}
  NamedRerouteDeclaration(None=@n7.Outline) {name=Outline, nodeColor=<rgba 0 0 0 1>}
  NamedRerouteDeclaration(None=MaterialFunctionCall(UVs=@n5, Size=@n6, OutlineThickness=Constant() {r=0.02}) {materialFunction=&MF_UI_Circle}.Outline) {name=AccentColor, nodeColor=<rgba 0 0 0 1>}
  n8 = TextureCoordinate()
  n9 = Add(A=ComponentMask(None=@n8) {r=true}) {constB=0.015}
  n10 = Add(A=ComponentMask(None=@n8) {g=true}, B=ScalarParameter() {defaultValue=0.06, parameterName=Icon Y pos}) {constB=0.06}
  n11 = ScalarParameter() {defaultValue=0.85, parameterName=Icon Width}
  n12 = ScalarParameter() {defaultValue=0.85, parameterName=Icon Height}
  n13 = TextureSample(UVs=MaterialFunctionCall(UVs=AppendVector(A=@n9, B=@n10), Texture Scale=AppendVector(A=@n11, B=@n12)) {materialFunction=&ScaleUVsByCenter}.UVs) {samplerSource=SSM_Clamp_WorldGroupSettings, texture=/SonicSpearGameplay/UI/Textures/T_UI_SonicSpear_Character}
  NamedRerouteDeclaration(None=@n13.RGB) {name=IconColor, nodeColor=<rgba 0 0 0 1>}
  NamedRerouteDeclaration(None=@n13.A) {name=IconOpacity, nodeColor=<rgba 0 0 0 1>}
  n14 = TextureCoordinate()
  n15 = Constant() {r=-1}
  n16 = Constant() {r=1}
  n17 = Constant() {r=0.025}
  NamedRerouteDeclaration(None=Add(
    A=Subtract(A=@n14) {constB=0}
    B=MaterialFunctionCall!(UV, SpeedX=@n15, SpeedY=@n16, Intensity=@n17) {materialFunction=/SonicSpearGameplay/UI/Material_Functions/MF_UI_WildEstate_Distort_2}.Result
   )) {name=Distort UV, nodeColor=<rgba 0.2422 0 1 1>}
 ]
}
