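# Translucent UI material for the Sonic Spear minimap marker, drawing an icon with a blinking, colored circular alert area.
# The icon texture is scaled around its center, and its squared RGB is blended with a tintable pulse color.
# MF_UI_Circle calls build the circle area, a faded mask and an outline.
# A time-driven sine provides the blink, and named reroutes organize the mask and final color composition.
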Material /SonicSpearGameplay/UI/Materials/M_UI_SonicSpear_Minimap
graph {
 properties {bAutomaticallySetUsageInEditor=false, blendMode=BLEND_Translucent, materialDomain=MD_UI, usedShadingModels=Default Lit}
 nodes [
  n1 = TextureSampleParameter2D(UVs=MaterialFunctionCall(UVs=TextureCoordinate(), Texture Scale=ScalarParameter() {defaultValue=0.15, parameterName=Texture scale}) {materialFunction=&ScaleUVsByCenter}.UVs) {parameterName=Texture, samplerSource=SSM_Clamp_WorldGroupSettings, texture=/SonicSpearGameplay/UI/Textures/T_UI_SonicSpear_Minimap_Icon}
  NamedRerouteDeclaration(None=@n1.A) {name=Texture Alpha, nodeColor=<rgba 1 0.0938 0 1>}
  n2 = NamedRerouteUsage()
  n3 = NamedRerouteUsage()
  n4 = Multiply(A=OneMinus(None=@n2), B=Multiply(A=NamedRerouteUsage(), B=@n3))
  n5 = NamedRerouteUsage()
  NamedRerouteDeclaration(None=Add(A=LinearInterpolate(A=@n4, B=@n2, Alpha=@n2), B=Multiply(A=@n5, B=@n3))) {name=Mask, nodeColor=<rgba 0 0.2266 1 1>}
  NamedRerouteDeclaration(None=Multiply(A=@n1.RGB, B=@n1.RGB)) {name=Texture color, nodeColor=<rgba 1 0.7031 0 1>}
  n6 = VectorParameter() {parameterName=Pulse Color, defaultValue=<rgba 1 0.1329 0.4179 1>}
  n7 = NamedRerouteUsage()
  n8 = OneMinus(None=@n7)
  NamedRerouteDeclaration(None=LinearInterpolate(A=Multiply(A=@n6.RGB, B=@n8), B=NamedRerouteUsage(), Alpha=@n7)) {name=final color, nodeColor=<rgba 0.3828 0 1 1>}
  n9 = ScalarParameter() {defaultValue=1, parameterName=Circle Scale}
  n10 = NamedRerouteUsage()
  NamedRerouteDeclaration(None=Multiply(A=MaterialFunctionCall(Size=@n9) {materialFunction=&MF_UI_Circle}.Fill, B=OneMinus(None=@n10))) {name=Circle area, nodeColor=<rgba 1 0 0.9141 1>}
  NamedRerouteDeclaration(None=MaterialFunctionCall(Size=Constant() {r=0.25}, GlowMax=Constant() {r=1}) {materialFunction=&MF_UI_Circle}.Fill) {name=Circle faded mask, nodeColor=<rgba 0 1 0.7656 1>}
  n11 = Time()
  n12 = ScalarParameter() {defaultValue=0.5, parameterName=Alert Speed}
  n13 = Add(A=Sine(None=Multiply(A=@n11, B=@n12))) {constB=0.5}
  NamedRerouteDeclaration(None=LinearInterpolate(A=ScalarParameter() {defaultValue=0.8, parameterName=Alert intensity}, Alpha=Multiply(A=@n13)) {constA=0.2}) {name=Blink FX, nodeColor=<rgba 0 0 0 1>}
  NamedRerouteDeclaration(None=MaterialFunctionCall(
    Size=ScalarParameter() {defaultValue=0.99, parameterName=Outline Scale}
    OutlineThickness=ScalarParameter() {defaultValue=0.01, parameterName=Outline thickness}
   ) {materialFunction=&MF_UI_Circle}.Outline) {name=Outline, nodeColor=<rgba 0.875 0 1 1>}
  NamedRerouteUsage()
  NamedRerouteUsage()
 ]
}
