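# WBP_SonicSpearAlert_Prioritized is a prioritized HUD widget wrapper around the Sonic Spear alert widget, built on AthenaCommonPrioritizedWidget.
# On construct it waits one tick, then flags itself as having active content through the prioritized widget interface.
# On destruct it clears that flag, so the alert only claims its priority slot while it exists.
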WidgetBlueprint /SonicSpearGameplay/UI/WBP_SonicSpearAlert_Prioritized
graphs [EventGraph {nodes [
   Event Construct()
   > DelayUntilNextTick!(execute) -> (then)
   > SetHasActiveContent!(execute, self:Fort Prioritized Widget Interface, bNewActive:bool=true) -> (then)
   Event Destruct()
   > SetHasActiveContent(bNewActive=false)
   Event PreConstruct()
  ]}]
variables []
widgets {
 info {
  inheritedWidgetCount=1
  parentClass=/Script/FortniteUI.AthenaCommonPrioritizedWidget
  rootWidgetClass=&/WBP_SonicSpearAlert.WBP_SonicSpearAlert_C
  widgetCount=2
 }
 widgets [
  WBP_SonicSpearAlert {widgetClassPath=&/WBP_SonicSpearAlert, bIsVariable=true, widget=WBP_SonicSpearAlert}
  TutorialHighlight {widgetClassPath=&TutorialHighlight}
 ]
}
